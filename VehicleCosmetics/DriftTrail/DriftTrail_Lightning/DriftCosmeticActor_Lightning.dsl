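; Cosmetic actor Blueprint for the Lightning vehicle drift trail, built around the NS_DriftTrail effect.
; When the color-ready event fires, it passes the color name, painted variables, bIsPainted key and additional info to a shared helper that sets the painted trail base.
; A second shared helper then sets material instances on the effect from that info, using the materials array and variable names.
; BeginPlay, overlap and Tick are empty stubs.
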(event EventBeginPlay
  (|CallParentFunction) ; This node is disabled and will not be called.\nDrag off pins to build functionality.
)

(event Collision|EventActorBeginOverlap
  (params (OtherActor :type Actor%20Object%20Reference)))

(event EventTick
  (params (DeltaSeconds :type Float%20%28single-precision%29)))

(event Vehicle|Cosmetics|EventOnColorReady
  (params (Color :type Name :default "None") (AdditionalInfo :type Vehicle%20Cosmetics%20Additional%20Parameter%20Infos%20Structure))
  (bind _ns_drifttrail (Variables|Default|GetNS_DriftTrail))
  (Class|BPVehicleCosmeticsFunctions|BPSetPaintedDriftTrailBase Color _ns_drifttrail "bIsPainted" (Variables|Default|GetPainted%20Variables) AdditionalInfo)
  (Class|BPVehicleCosmeticsFunctions|SetMICBasedonValue AdditionalInfo _ns_drifttrail (Variables|Default|GetMaterials%20Array) (Variables|Default|GetVariables%20Name)))

(fn ConstructionScript
  (|CallParentFunction))
