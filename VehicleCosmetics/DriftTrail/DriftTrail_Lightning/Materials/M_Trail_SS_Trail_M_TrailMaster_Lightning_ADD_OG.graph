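# Translucent unlit master material for lightning drift-trail effects on Niagara ribbons and mesh particles.
# Time-driven flipbook layers select texture channels, then add, multiply or lerp via static switches; a third layer's combinations feed nothing.
# An optional panning mask modulates the result before U-axis fades, particle alpha and optional depth fade shape opacity.
# Particle color and dynamic parameters for brightness, blend, saturation and color shift drive emissive through MF_FX_EmissiveAdjustments.
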Material /VehicleCosmetics/DriftTrail/DriftTrail_Lightning/Materials/M_Trail_SS_Trail_M_TrailMaster_Lightning_ADD_OG
graph {
 properties {
  bUsedWithNiagaraMeshParticles=true
  bUsedWithNiagaraRibbons=true
  blendMode=BLEND_Translucent
  pixelDepthOffsetMode=PDOM_Legacy
  shadingModel=MSM_Unlit
  twoSided=true
  usedShadingModels=Unlit
 }
 nodes [
  n1 = TextureCoordinate() {coordinateIndex=1, vTiling=0}
  n2 = TextureCoordinate() {coordinateIndex=1, uTiling=0}
  n3 = AppendVector(
   A=Multiply(A=ComponentMask(None=@n1) {r=true}, B=ScalarParameter() {defaultValue=1, group=Details, parameterName=TileU_03})
   B=Multiply(A=ComponentMask(None=@n2) {g=true}, B=ScalarParameter() {defaultValue=1, group=Details, parameterName=TileV_03})
  )
  n4 = StaticSwitchParameter(True=Rotator(Coordinate=@n3, Time=Constant() {r=0.5}) {speed=3.14}, False=@n3) {group=FlipBook, parameterName=RotateFlipbook}
  n5 = ScalarParameter() {defaultValue=4, group=FlipBook, parameterName=Columns_03}
  n6 = ScalarParameter() {defaultValue=4, group=FlipBook, parameterName=Rows_03}
  n7 = AppendVector(A=@n5, B=@n6)
  n8 = Multiply(A=@n5, B=@n6)
  n9 = Frac(None=Multiply(A=Time(), B=ScalarParameter() {defaultValue=1, group=FlipBook, parameterName=Speed_03}))
  n10 = Floor(None=Multiply(A=AppendVector(A=@n8, B=@n6), B=AppendVector(A=@n9, B=@n9)))
  n11 = TextureCoordinate() {coordinateIndex=1}
  n12 = Constant() {r=0.5}
  n13 = StaticSwitchParameter(
   True=Add(A=Divide(A=@n4, B=@n7), B=Divide(A=@n10, B=@n7))
   False=StaticSwitchParameter(True=Rotator(Coordinate=@n11, Time=@n12) {speed=3.14}, False=@n11) {group=FlipBook, parameterName=RotateFlipbook}
  ) {defaultValue=true, group=FlipBook, parameterName=FlipBook}
  n14 = Multiply(
   A=ScalarParameter() {defaultValue=1, group=Blend, parameterName=Blend_03_Brightness}
   B=StaticComponentMaskParameter(None=TextureSampleParameter2D(UVs=@n13) {group=FlipBook, parameterName=BaseTexture_03, texture=/DealershipSharedLibrary/FX/Textures/Gradient_Masks/T_VCVFX_BasicTrail_A}.RGB) {defaultB=true, group=Blend, parameterName=Blend_03}
  )
  n15 = TextureCoordinate() {coordinateIndex=1, vTiling=0}
  n16 = TextureCoordinate() {coordinateIndex=1, uTiling=0}
  n17 = AppendVector(
   A=Multiply(A=ComponentMask(None=@n15) {r=true}, B=ScalarParameter() {defaultValue=1, group=Details, parameterName=TileU})
   B=Multiply(A=ComponentMask(None=@n16) {g=true}, B=ScalarParameter() {defaultValue=1, group=Details, parameterName=TileV})
  )
  n18 = Constant() {r=0.5}
  n19 = ScalarParameter() {defaultValue=4, group=FlipBook, parameterName=Columns}
  n20 = ScalarParameter() {defaultValue=4, group=FlipBook, parameterName=Rows}
  n21 = AppendVector(A=@n19, B=@n20)
  n22 = Divide(
   A=StaticSwitchParameter(True=Rotator(Coordinate=@n17, Time=@n18) {speed=3.14}, False=@n17) {group=FlipBook, parameterName=RotateFlipbook}
   B=@n21
  )
  n23 = AppendVector(A=Multiply(A=@n19, B=@n20), B=@n20)
  n24 = Frac(None=Multiply(A=Time(), B=ScalarParameter() {defaultValue=1, group=FlipBook, parameterName=Speed}))
  n25 = AppendVector(A=@n24, B=@n24)
  n26 = Divide(A=Floor(None=Multiply(A=@n23, B=@n25)), B=@n21)
  n27 = TextureCoordinate() {coordinateIndex=1}
  n28 = Rotator(Coordinate=@n27, Time=Constant() {r=0.5}) {speed=3.14}
  n29 = TextureSampleParameter2D(UVs=StaticSwitchParameter(True=Add(A=@n22, B=@n26), False=StaticSwitchParameter(True=@n28, False=@n27) {group=FlipBook, parameterName=RotateFlipbook}) {defaultValue=true, group=FlipBook, parameterName=FlipBook}) {group=FlipBook, parameterName=BaseTexture, texture=&T_VCVFX_BasicTrail_A}
  n30 = Multiply(
   A=ScalarParameter() {defaultValue=1, group=Blend, parameterName=Blend_01_Brightness}
   B=StaticComponentMaskParameter(None=@n29.RGB) {defaultG=true, group=Blend, parameterName=Blend_01}
  )
  n31 = TextureCoordinate() {coordinateIndex=1, vTiling=0}
  n32 = TextureCoordinate() {coordinateIndex=1, uTiling=0}
  n33 = AppendVector(
   A=Multiply(A=ComponentMask(None=@n31) {r=true}, B=ScalarParameter() {defaultValue=1, group=Details, parameterName=TileU_02})
   B=Multiply(A=ComponentMask(None=@n32) {g=true}, B=ScalarParameter() {defaultValue=1, group=Details, parameterName=TileV_02})
  )
  n34 = Constant() {r=0.5}
  n35 = ScalarParameter() {defaultValue=4, group=FlipBook, parameterName=Columns_02}
  n36 = ScalarParameter() {defaultValue=4, group=FlipBook, parameterName=Rows_02}
  n37 = AppendVector(A=@n35, B=@n36)
  n38 = Divide(
   A=StaticSwitchParameter(True=Rotator(Coordinate=@n33, Time=@n34) {speed=3.14}, False=@n33) {group=FlipBook, parameterName=RotateFlipbook_02}
   B=@n37
  )
  n39 = AppendVector(A=Multiply(A=@n35, B=@n36), B=@n36)
  n40 = Frac(None=Multiply(A=Time(), B=ScalarParameter() {defaultValue=1, group=FlipBook, parameterName=Speed_02}))
  n41 = AppendVector(A=@n40, B=@n40)
  n42 = Divide(A=Floor(None=Multiply(A=@n39, B=@n41)), B=@n37)
  n43 = TextureCoordinate() {coordinateIndex=1}
  n44 = Rotator(Coordinate=@n43, Time=Constant() {r=0.5}) {speed=3.14}
  n45 = TextureSampleParameter2D(UVs=StaticSwitchParameter(True=Add(A=@n38, B=@n42), False=StaticSwitchParameter(True=@n44, False=@n43) {group=FlipBook, parameterName=RotateFlipbook_02}) {defaultValue=true, group=FlipBook, parameterName=Flipbook_02}) {group=FlipBook, parameterName=BaseTexture_02, texture=&T_VCVFX_BasicTrail_A}
  n46 = Multiply(
   A=ScalarParameter() {defaultValue=1, group=Blend, parameterName=Blend_02_Brightness}
   B=StaticComponentMaskParameter(None=StaticSwitchParameter(True=@n29.RGB, False=@n45.RGB) {defaultValue=true, group=Blend, parameterName=SingleTexture}) {defaultB=true, group=Blend, parameterName=Blend_02}
  )
  n47 = Add(A=@n30, B=@n46)
  Add(A=@n14, B=@n47)
  MaterialExpressionVertexColor_1 = ParticleColor()
  n48 = ScalarParameter() {group=Pan, parameterName=PanTile}
  n49 = Multiply(A=TextureCoordinate() {uTiling=0.05, vTiling=0.25}, B=@n48)
  n50 = Multiply(A=Time(), B=ScalarParameter() {defaultValue=1, group=Pan, parameterName=PanSpeed})
  n51 = TextureCoordinate() {uTiling=0.25, vTiling=0.05}
  n52 = StaticSwitchParameter(
   True=Rotator(Coordinate=Panner(Coordinate=@n49, Time=@n50) {speedY=0.4}, Time=Constant() {r=0.5}) {speed=3.14}
   False=Panner(Coordinate=Multiply(A=@n51, B=@n48), Time=@n50) {speedX=0.4}
  ) {group=FlipBook, parameterName=RotateFlipbook}
  n53 = DynamicParameter() {paramNames [Brightness,LerpBlend,Saturation,ColorShift]}
  n54 = Multiply(A=@n30, B=@n46)
  n55 = StaticSwitchParameter(
   True=StaticSwitchParameter(
    True=LinearInterpolate(A=@n30, B=@n46, Alpha=@n53.LerpBlend)
    False=StaticSwitchParameter(True=@n54, False=@n47) {group=Blend, parameterName="BlendType(T=M)(F=A)"}
   ) {group=Blend, parameterName=LerpBlend}
   False=Multiply(
    A=StaticComponentMaskParameter(None=@n29.RGB) {defaultG=true, group=Blend, parameterName=RGB_Mask}
    B=ScalarParameter() {defaultValue=1, group=Blend, parameterName=RGB_Mask_Brightness}
   )
  ) {group=Blend, parameterName=Blend}
  n56 = Multiply(
   A=Multiply(
    A=ScalarParameter() {defaultValue=3, group=Pan, parameterName=PanBrightness}
    B=TextureSample(UVs=@n52) {texture=/DealershipSharedLibrary/FX/Textures/Gradient_Masks/T_VCVFX_TimeMask_A}.RGB
   )
   B=@n55
  )
  n57 = StaticSwitchParameter(True=Add(A=@n56, B=@n55), False=@n56) {group=Pan, parameterName=AddPan?}
  n58 = TextureCoordinate() {uTiling=4, vTiling=4}
  n59 = ComponentMask(None=StaticSwitchParameter(True=TextureCoordinate(), False=Frac(None=@n58)) {defaultValue=true, group=FlipBook, parameterName=FlipBook}) {r=true}
  n60 = Constant() {r=1}
  n61 = OneMinus(None=@n59)
  n62 = Multiply(
   A=Multiply(
    A=StaticSwitchParameter(True=@n57, False=@n55) {group=Pan, parameterName=PanLightning}
    B=StaticSwitchParameter(True=@n59, False=@n60) {defaultValue=true, parameterName=FadeIn?}
   )
   B=@n61
  )
  n63 = Multiply(A=@MaterialExpressionVertexColor_1.A, B=@n62)
  MaterialExpressionConstantClamp_5 = Saturate(None=StaticSwitchParameter(True=DepthFade(Opacity=@n63) {fadeDistanceDefault=255}, False=@n63) {group=Opacity, parameterName=DBA?})
  MaterialExpressionConstantClamp_7 = Saturate(None=Multiply(A=Constant() {r=0.15}, B=@MaterialExpressionVertexColor_1.RGB))
  MaterialExpressionConstantClamp_6 = Saturate(None=Add(A=LinearInterpolate(A=@MaterialExpressionVertexColor_1.RGB, B=@MaterialExpressionConstantClamp_7, Alpha=@n53.ColorShift), B=@n53.Saturation))
  MaterialExpressionConstantClamp_9 = Saturate(None=@n61)
  MaterialExpressionConstantClamp_8 = Saturate(None=@n62)
  n64 = Multiply(A=Multiply(A=@MaterialExpressionConstantClamp_9, B=Constant() {r=8}), B=@MaterialExpressionConstantClamp_8)
  MaterialFunctionCall(Emissive=Multiply(A=@MaterialExpressionConstantClamp_6, B=Multiply(A=@n64, B=@n53.Brightness))) {materialFunction=&MF_FX_EmissiveAdjustments}
  Multiply(A=@n14, B=@n54)
 ]
}
